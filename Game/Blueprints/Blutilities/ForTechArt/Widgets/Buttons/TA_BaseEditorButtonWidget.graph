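# Base button widget for tech-art editor tools, with no layout of its own and stub getters for background, label and icon widgets.
# Mouse and drag handlers track hover and press state, raising click, double-click and selection events, with optional toggle selection.
# List item data supplies label text and alignment.
# State changes restyle the supplied widgets through validateVisualsNextTick, which validates immediately despite its name, leaving the Tick deferral flag unset.
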WidgetBlueprint /Game/Blueprints/Blutilities/ForTechArt/Widgets/Buttons/TA_BaseEditorButtonWidget
graphs [
 EventGraph {nodes [
   Event PreConstruct()
   > setLabel!(execute, self, label:String=Get label() -> (label:String)) -> (then)
   > validateVisuals!(execute, self) -> (then, " ":bool)
   n1 = Event setListItemData!() -> (OutputDelegate, then, listItemData:Object Reference) {is=title K2Node_Event}
   > handleNewListDataSet!(execute, self, currentListItemData:Object Reference=@n1.listItemData) -> (then)
   n2 = Event setListItemSelection!() -> (OutputDelegate, then, selected:bool) {is=title K2Node_Event}
   > setSelected!(execute, self, selected:bool=@n2.selected, throwEvent:bool=true) -> (then)
   n3 = Event setOwningList!() -> (OutputDelegate, then, list:User Widget Object Reference) {is=title K2Node_Event}
   > Set owningList(owningList:User Widget Object Reference=@n3.list) -> (Output_Get:User Widget Object Reference)
   Event Tick()
   > Branch(Condition=Get doVisualsNeedValidation() -> (doVisualsNeedValidation:bool))
   > Set doVisualsNeedValidation(doVisualsNeedValidation:bool=false) -> (Output_Get:bool)
   > validateVisuals()
   Event OnMouseEnter()
   > Set isMouseOver(isMouseOver:bool=true) -> (Output_Get:bool)
   > validateVisualsNextTick()
   Event OnMouseLeave!() -> (OutputDelegate, then, MouseEvent:"Pointer Event Structure (by ref)") {is=title K2Node_Event}
   > Set isMouseOver(isMouseOver=false)
   > Set isPressed(isPressed:bool=false) -> (Output_Get:bool)
   > validateVisualsNextTick()
   n4 = Event OnDragCancelled!() -> (OutputDelegate, then, PointerEvent:"Pointer Event Structure (by ref)", Operation:Drag Drop Operation Object Reference) {is=title K2Node_Event}
   > handleRelease!(execute, self, InputEvent:"Pointer Event Structure (by ref)"=@n4.PointerEvent) -> (then, " wasClicked":bool)
   Event OnDragEnter!() -> (OutputDelegate, then, MyGeometry:Geometry Structure, PointerEvent:Pointer Event Structure, Operation:Drag Drop Operation Object Reference) {is=title K2Node_Event}
   > Set isMouseOver(isMouseOver=true)
   > validateVisualsNextTick()
   Event OnDragLeave!() -> (OutputDelegate, then, PointerEvent:Pointer Event Structure, Operation:Drag Drop Operation Object Reference) {is=title K2Node_Event}
   > Set isMouseOver(isMouseOver=false)
   > validateVisualsNextTick()
   n5 = Event On Item Selection Changed!() -> (OutputDelegate, then, bIsSelected:bool) {is=title K2Node_Event}
   > setSelected(selected=@n5.bIsSelected, throwEvent=true)
   n6 = Event OnListItemObjectSet()
   > handleNewListDataSet(currentListItemData=@n6.ListItemObject)
   ForceLayoutPrepass(self:Widget Object Reference)
   Event Construct()
  ]}
 FilterBySearchString {nodes [
   FilterBySearchString!() -> (then, SearchString:String) {is=title K2Node_FunctionEntry}
   > Return Node!(execute, ShouldDisplay:bool=false)
  ]}
 OnMouseButtonDoubleClick {nodes [
   n1 = OnMouseButtonDoubleClick!() -> (then, InMyGeometry:Geometry Structure, InMouseEvent:"Pointer Event Structure (by ref)") {is=title K2Node_FunctionEntry}
   > Branch(Condition=Is Mouse Button Down!(Input:"Pointer Event Structure (by ref)"=@n1.InMouseEvent, MouseButton:Key Structure=LeftMouseButton) -> (ReturnValue:bool))
   > n2 = handleDoubleClicked!(execute, self) -> (then, allowInputCapture:bool)
   > n3 = Branch(Condition=@n2)
   n4 = Get currentListItemData() -> (currentListItemData:Object Reference)
   > Return Node!(
    execute
    ReturnValue:Event Reply Structure=Select(Option 0:Event Reply Structure=Handled(), Option 1:Event Reply Structure=Unhandled(), Index=IsValid(Object=@n4)) -> (ReturnValue:Event Reply Structure)
   )
   Return Node(execute=@n3.else, ReturnValue=Unhandled())
  ]}
 OnMouseButtonDown {nodes [
   n1 = OnMouseButtonDown!() -> (then, MyGeometry:Geometry Structure, MouseEvent:"Pointer Event Structure (by ref)") {is=title K2Node_FunctionEntry}
   > Branch(Condition=Is Mouse Button Down(Input=@n1.MouseEvent, MouseButton=LeftMouseButton))
   > Set isPressed(isPressed=true)
   > validateVisualsNextTick()
   > DetectDragIfPressed!(execute, PointerEvent:"Pointer Event Structure (by ref)"=@n1.MouseEvent, DragKey:Key Structure=LeftMouseButton) -> (then, ReturnValue:Event Reply Structure)
   n2 = Get currentListItemData()
   > Return Node(ReturnValue=Select(Option 0=Handled(), Option 1=Unhandled(), Index=IsValid(Object=@n2)))
  ]}
 OnMouseButtonUp {nodes [
   n1 = OnMouseButtonUp!() -> (then, MyGeometry:Geometry Structure, MouseEvent:"Pointer Event Structure (by ref)") {is=title K2Node_FunctionEntry}
   > handleRelease(InputEvent=@n1.MouseEvent)
   n2 = Get currentListItemData()
   > Return Node(ReturnValue=Select(Option 0=Handled(), Option 1=Unhandled(), Index=IsValid(Object=@n2)))
  ]}
 getBackgroundWidget {nodes [
   getBackgroundWidget() {is=title K2Node_FunctionEntry}
   > Return Node!(execute, backgroundWidget:Widget Object Reference)
  ]}
 getIconWidget {nodes [
   getIconWidget() {is=title K2Node_FunctionEntry}
   > Return Node!(execute, iconWidget:Widget Object Reference)
  ]}
 getLabelWidget {nodes [
   getLabelWidget() {is=title K2Node_FunctionEntry}
   > Return Node!(execute, labelWidget:Widget Object Reference)
  ]}
 getSelected {nodes [
   getSelected() {is=title K2Node_FunctionEntry}
   > Return Node!(execute, isSelected:bool=Get isSelected() -> (isSelected:bool))
  ]}
 getValidatedButtonBrush {nodes [
   getValidatedButtonBrush() {is=title K2Node_FunctionEntry}
   > Set l_style(l_style:Button Style Structure=Select(
     Option 0:Button Style Structure=Get ButtonStyle_Default() -> (ButtonStyle_Default:Button Style Structure)
     Option 1:Button Style Structure=Get ButtonStyle_Selected() -> (ButtonStyle_Selected:Button Style Structure)
     Index=Get isSelected()
    ) -> (ReturnValue:Button Style Structure)) -> (Output_Get:Button Style Structure)
   n1 = Get!() -> (
    l_style_Normal:Slate Brush Structure
    l_style_Hovered:Slate Brush Structure
    l_style_Pressed:Slate Brush Structure
    l_style_Disabled:Slate Brush Structure
    l_style_NormalForeground:Slate Color Structure
    l_style_HoveredForeground:Slate Color Structure
    l_style_PressedForeground:Slate Color Structure
    l_style_DisabledForeground:Slate Color Structure
    l_style_NormalPadding:Margin Structure
    l_style_PressedPadding:Margin Structure
    l_style_PressedSlateSound:Slate Sound Structure
    l_style_ClickedSlateSound:Slate Sound Structure
    l_style_HoveredSlateSound:Slate Sound Structure
   )
   n2 = Select(Option 0:Slate Brush Structure=@n1, Option 1:Slate Brush Structure=@n1.l_style_Hovered, Index=Get isMouseOver() -> (isMouseOver:bool)) -> (ReturnValue:Slate Brush Structure)
   n3 = Get isPressed() -> (isPressed:bool)
   > Return Node!(
    execute
    brush:Slate Brush Structure=Select(
     Option 0=@n1.l_style_Disabled
     Option 1=Select(Option 0=@n2, Option 1=@n1.l_style_Pressed, Index=@n3)
     Index=Get isEnabled() -> (isEnabled:bool)
    )
   )
  ]}
 getValidatedButtonColor {nodes [
   getValidatedButtonColor() {is=title K2Node_FunctionEntry}
   > Return Node!(
    execute
    buttonColor_Default:Linear Color Structure=Select(
     Option 0:Linear Color Structure=Get buttonColor_Default() -> (buttonColor_Default:Linear Color Structure)
     Option 1:Linear Color Structure=Get buttonColor_Selected() -> (buttonColor_Selected:Linear Color Structure)
     Index=getSelected!() -> (isSelected:bool) {is=title K2Node_CallFunction}
    ) -> (ReturnValue:Linear Color Structure)
   )
  ]}
 handleClicked {nodes [
   n1 = handleClicked!() -> (then, InputEvent:"Pointer Event Structure (by ref)") {is=title K2Node_FunctionEntry}
   > n2 = Sequence()
   > Branch(Condition=Get autoToggleSelection() -> (autoToggleSelection:bool))
   > setSelected(selected=NOT Boolean(A=Get isSelected()), throwEvent=true)
   Call On Clicked!(
    execute:Exec=@n2.then_1
    self:TA Base Editor Button Widget Object Reference
    widget:User Widget Object Reference=Self-Reference()
    event:Pointer Event Structure=@n1
   ) -> (then:Exec) {is=title K2Node_CallDelegate}
  ]}
 handleDoubleClicked {nodes [
   handleDoubleClicked() {is=title K2Node_FunctionEntry}
   > Call On Double Clicked!(execute:Exec, self:TA Base Editor Button Widget Object Reference, widget:User Widget Object Reference=Self-Reference()) -> (then:Exec) {is=title K2Node_CallDelegate}
   > Return Node!(execute, allowInputCapture:bool=true)
  ]}
 handleNewListDataSet {nodes [
   n1 = handleNewListDataSet!() -> (then, currentListItemData:Object Reference) {is=title K2Node_FunctionEntry}
   > n2 = Set currentListItemData(currentListItemData:Object Reference=@n1) -> (Output_Get:Object Reference)
   > updateAfterDataChange!(execute, self, currentListItemData:Object Reference=@n2) -> (then)
  ]}
 handleRelease {nodes [
   n1 = handleRelease!() -> (then, InputEvent:"Pointer Event Structure (by ref)") {is=title K2Node_FunctionEntry}
   > n2 = Branch(Condition=Get isPressed())
   > Set isPressed(isPressed=false)
   > handleClicked!(execute, self, InputEvent:"Pointer Event Structure (by ref)"=@n1) -> (then) {is=title K2Node_CallFunction}
   > validateVisualsNextTick()
   > Return Node!(execute, " wasClicked":bool=true)
   Return Node(execute=@n2.else, " wasClicked"=false)
  ]}
 onClicked {nodes [onClicked!() -> (then, widget:User Widget Object Reference, event:Pointer Event Structure) {is=title K2Node_FunctionEntry}]}
 onDoubleClicked {nodes [onDoubleClicked!() -> (then, widget:User Widget Object Reference) {is=title K2Node_FunctionEntry}]}
 onSelectionChanged {nodes [onSelectionChanged!() -> (then, selected:bool=~, widget:User Widget Object Reference) {is=title K2Node_FunctionEntry}]}
 setHorizontalAlignmentSet {nodes [
   n1 = setHorizontalAlignmentSet!() -> (then, alignment:EHorizontalAlignment Enum=HAlign_Fill) {is=title K2Node_FunctionEntry}
   > n2 = Cast To TextBlock!(execute, Object=getLabelWidget!() -> (labelWidget:Widget Object Reference) {is=title K2Node_CallFunction}) -> (then, CastFailed, AsText:Text Object Reference) {is=title K2Node_DynamicCast}
   > SetJustification!(
    execute
    self:Text Layout Widget Object Reference=@n2
    InJustification:ETextJustify Enum=Select!(
     HAlign_Fill:ETextJustify Enum=Center
     HAlign_Left:ETextJustify Enum=Left
     HAlign_Center:ETextJustify Enum=Center
     HAlign_Right:ETextJustify Enum=Right
     Index:EHorizontalAlignment Enum=@n1
    ) -> (ReturnValue:ETextJustify Enum)
   ) -> (then)
  ]}
 setLabel {nodes [
   n1 = setLabel!() -> (then, label:String) {is=title K2Node_FunctionEntry}
   > Set label(label:String=@n1) -> (Output_Get:String)
   > n2 = Sequence()
   > n3 = Cast To CommonTextBlock!(execute, Object=getLabelWidget()) -> (then, CastFailed, AsCommon Text:Common Text Object Reference) {is=title K2Node_DynamicCast}
   n4 = "To Text (String)"(InString=Get label())
   > "SetText (Text)"(self=@n3, InText=@n4)
   n5 = Cast To TextBlock(execute=@n3.CastFailed, Object=getLabelWidget())
   > "SetText (Text)"(self=@n5, InText=@n4)
   n6 = Cast To RichTextBlock!(execute=@n5.CastFailed, Object=getLabelWidget()) -> (then, CastFailed, AsRich Text Block:Rich Text Block Object Reference) {is=title K2Node_DynamicCast}
   > SetText(self=@n6, InText=@n4)
   validateVisualsNextTick(execute=@n2.then_1)
  ]}
 setSelected {nodes [
   n1 = setSelected!() -> (then, selected:bool=~, throwEvent:bool=true) {is=title K2Node_FunctionEntry}
   > Branch(Condition="Not Equal (Boolean)"!(A:bool=@n1, B:bool=Get isSelected()) -> (ReturnValue:bool) {is=title K2Node_CallFunction})
   > Set isSelected(isSelected:bool=@n1) -> (Output_Get:bool)
   > validateVisualsNextTick()
   > Branch(Condition=@n1.throwEvent)
   > Call On Selection Changed!(
    execute:Exec
    self:TA Base Editor Button Widget Object Reference
    selected:bool=Get isSelected()
    widget:User Widget Object Reference=Self-Reference()
   ) -> (then:Exec) {is=title K2Node_CallDelegate}
  ]}
 updateAfterDataChange {nodes [
   n1 = updateAfterDataChange!() -> (then, currentListItemData:Object Reference) {is=title K2Node_FunctionEntry}
   > n2 = Cast To TA_GenericListItemData!(execute, Object=@n1) -> (then, CastFailed, AsTA Generic List Item Data:TA Generic List Item Data Object Reference) {is=title K2Node_DynamicCast}
   > setLabel(label=Get displayString!(self:TA Generic List Item Data Object Reference=@n2) -> (displayString:String)) {is=title K2Node_CallFunction}
   > setHorizontalAlignmentSet!(
    execute
    self
    alignment:EHorizontalAlignment Enum=Get labelAlignment!(self:TA Generic List Item Data Object Reference=@n2) -> (labelAlignment:EHorizontalAlignment Enum=HAlign_Fill)
   ) -> (then) {is=title K2Node_CallFunction}
  ]}
 validateBackgroundColor {nodes [
   validateBackgroundColor() {is=title K2Node_FunctionEntry}
   > Is Valid(InputObject=getBackgroundWidget!() -> (backgroundWidget:Widget Object Reference) {is=title K2Node_CallFunction})
   > n1 = Sequence()
   > n2 = Sequence!(execute) -> (then_0, then_1, then_2:Exec, then_3:Exec)
   > Set l_brush(l_brush:Slate Brush Structure=getValidatedButtonBrush!() -> (brush:Slate Brush Structure) {is=title K2Node_CallFunction}) -> (Output_Get:Slate Brush Structure)
   n3 = Cast To Border!(execute=@n2.then_1, Object=getBackgroundWidget()) -> (then, CastFailed, AsBorder:Border Object Reference) {is=title K2Node_DynamicCast}
   > SetBrush(self:Border Object Reference=@n3, InBrush=Get l_brush() -> (l_brush:Slate Brush Structure))
   > SetBrushColor(self=@n3, InBrushColor=getValidatedButtonColor!() -> (buttonColor_Default:Linear Color Structure) {is=title K2Node_CallFunction})
   > Return Node!(execute, handled:bool=true)
   n4 = Cast To Image!(execute=@n2.then_2, Object=getBackgroundWidget()) -> (then, CastFailed, AsImage:Image Object Reference) {is=title K2Node_DynamicCast}
   > SetBrush(self:Image Object Reference=@n4, InBrush=Get l_brush())
   > SetColorAndOpacity(self:Image Object Reference=@n4, InColorAndOpacity:Linear Color Structure=getValidatedButtonColor())
   > Return Node(handled=true)
   Return Node(execute=@n1.then_1, handled=false)
  ]}
 validateIconVisuals {nodes [
   validateIconVisuals() {is=title K2Node_FunctionEntry}
   > Is Valid(InputObject=getIconWidget!() -> (iconWidget:Widget Object Reference) {is=title K2Node_CallFunction})
   > n1 = Is Valid(InputObject=Break Slate Brush!(SlateBrush:"Slate Brush Structure (by ref)"=Get iconStyle() -> (iconStyle:Slate Brush Structure)) -> (
     TintColor:Slate Color Structure
     DrawAs:ESlateBrushDrawType Enum
     Tiling:ESlateBrushTileType Enum
     Mirroring:ESlateBrushMirrorType Enum
     ImageSize:Vector 2D Structure
     Margin:Margin Structure
     ResourceObject:Object Reference
     OutlineSettings:Slate Brush Outline Settings Structure
    ) {is=title K2Node_BreakStruct}.ResourceObject)
   SetVisibility(execute=@n1.Is Not Valid, self=getIconWidget(), InVisibility=Collapsed)
   n2 = Sequence!(execute=@n1.Is Valid) -> (then_0, then_1, then_2)
   > n3 = Cast To Image(Object=getIconWidget())
   > SetBrush(self=@n3, InBrush=Get iconStyle())
   > SetVisibility(self=@n3, InVisibility=SelfHitTestInvisible)
   > Return Node(handled=true)
   Return Node(execute=@n2.then_2, handled=false)
  ]}
 validateLabelColor {nodes [
   validateLabelColor() {is=title K2Node_FunctionEntry}
   > Is Valid(InputObject=getLabelWidget())
   > n1 = Sequence!(execute) -> (then_0, then_1)
   > n2 = Sequence!(execute) -> (then_0, then_1, then_2:Exec, then_3:Exec)
   > Set l_color(l_color:Linear Color Structure=Select!(
     Option 0:Linear Color Structure=Get labelColor_Default() -> (labelColor_Default:Linear Color Structure)
     Option 1:Linear Color Structure=Get labelColor_Selected() -> (labelColor_Selected:Linear Color Structure)
     Index:bool=Get isSelected()
    ) -> (ReturnValue:Linear Color Structure)) -> (Output_Get:Linear Color Structure)
   n3 = Cast To TextBlock(execute=@n2.then_1, Object=getLabelWidget())
   > SetColorAndOpacity!(
    execute
    self:Text Object Reference=@n3
    InColorAndOpacity_SpecifiedColor:Linear Color Structure=Get l_color() -> (l_color:Linear Color Structure)
    InColorAndOpacity_ColorUseRule:ESlateColorStylingMode Enum
   ) -> (then)
   n4 = Replace(SourceString=Get label(), From=" ")
   SetVisibility(
    execute=@n2.then_2
    self=getLabelWidget()
    InVisibility=Select(Option 0:ESlateVisibility Enum=Collapsed, Option 1:ESlateVisibility Enum=HitTestInvisible, Index="Not Equal Exactly (String)"(A=@n4)) -> (ReturnValue:ESlateVisibility Enum)
   )
   Return Node(execute=@n2.then_3, handled=false)
   Return Node(execute=@n1.then_1, handled=false)
  ]}
 validateVisuals {nodes [
   validateVisuals() {is=title K2Node_FunctionEntry}
   > validateBackgroundColor!(execute, self) -> (then, handled:bool) {is=title K2Node_CallFunction}
   > validateLabelColor!(execute, self) -> (then, handled:bool) {is=title K2Node_CallFunction}
   > validateIconVisuals!(execute, self) -> (then, handled:bool) {is=title K2Node_CallFunction}
   > Return Node!(execute, " ":bool=false)
  ]}
 validateVisualsNextTick {nodes [
   validateVisualsNextTick() {is=title K2Node_FunctionEntry}
   > n1 = Is Valid(InputObject=GetPlayerController!(PlayerIndex:int=0) -> (ReturnValue:Player Controller Object Reference))
   validateVisuals(execute=[@n1.Is Not Valid, @n1.Is Valid]) {is=title K2Node_CallFunction}
   Set doVisualsNeedValidation(doVisualsNeedValidation=true)
  ]}
]
variables [
 autoToggleSelection {type=boolean}
 buttonColor_Default {type=LinearColor, default=<rgba 1 1 1 1>}
 buttonColor_Selected {type=LinearColor, default=<rgba 1 1 1 1>}
 ButtonStyle_Default {
  type=ButtonStyle
  default {
   clickedSlateSound {resourceObject=null}
   disabled {
    bIsDynamicallyLoaded=false
    drawAs=NoDrawType
    imageType=NoImage
    mirroring=NoMirror
    resourceName=None
    resourceObject=null
    tiling=NoTile
    imageSize=<0 0>
    margin=<ltrb 0 0 0 0>
    outlineSettings {
     bUseBrushTransparency=false
     roundingType=HalfHeightRadius
     width=0
     color {colorUseRule=UseColor_Specified, specifiedColor=<rgba 0 0 0 0>}
     cornerRadii=<0 0 0 1>
    }
    tintColor {colorUseRule=UseColor_Specified, specifiedColor=<rgba 0.4167 0.4167 0.4167 1>}
    uVRegion {bIsValid=false, max=<0 0>, min=<0 0>}
   }
   disabledForeground {colorUseRule=UseColor_Foreground, specifiedColor=<rgba 1 0 1 1>}
   hovered {
    bIsDynamicallyLoaded=false
    drawAs=Image
    imageType=NoImage
    mirroring=NoMirror
    resourceName=None
    resourceObject=null
    tiling=NoTile
    imageSize=<32 32>
    margin=<ltrb 0 0 0 0>
    outlineSettings {
     bUseBrushTransparency=false
     roundingType=HalfHeightRadius
     width=0
     color {colorUseRule=UseColor_Specified, specifiedColor=<rgba 0 0 0 0>}
     cornerRadii=<0 0 0 1>
    }
    tintColor {colorUseRule=UseColor_Specified, specifiedColor=<rgba 0.8542 0.8542 0.8542 1>}
    uVRegion {bIsValid=false, max=<0 0>, min=<0 0>}
   }
   hoveredForeground {colorUseRule=UseColor_Foreground, specifiedColor=<rgba 1 0 1 1>}
   hoveredSlateSound {resourceObject=null}
   normal {
    bIsDynamicallyLoaded=false
    drawAs=Image
    imageType=NoImage
    mirroring=NoMirror
    resourceName=None
    resourceObject=null
    tiling=NoTile
    imageSize=<32 32>
    margin=<ltrb 0 0 0 0>
    outlineSettings {
     bUseBrushTransparency=false
     roundingType=HalfHeightRadius
     width=0
     color {colorUseRule=UseColor_Specified, specifiedColor=<rgba 0 0 0 0>}
     cornerRadii=<0 0 0 1>
    }
    tintColor {colorUseRule=UseColor_Specified, specifiedColor=<rgba 0.5457 0.5457 0.5457 1>}
    uVRegion {bIsValid=false, max=<0 0>, min=<0 0>}
   }
   normalForeground {colorUseRule=UseColor_Foreground, specifiedColor=<rgba 1 0 1 1>}
   normalPadding=<ltrb 0 0 0 0>
   pressed {
    bIsDynamicallyLoaded=false
    drawAs=Image
    imageType=NoImage
    mirroring=NoMirror
    resourceName=None
    resourceObject=null
    tiling=NoTile
    imageSize=<32 32>
    margin=<ltrb 0 0 0 0>
    outlineSettings {
     bUseBrushTransparency=false
     roundingType=HalfHeightRadius
     width=0
     color {colorUseRule=UseColor_Specified, specifiedColor=<rgba 0 0 0 0>}
     cornerRadii=<0 0 0 1>
    }
    tintColor {colorUseRule=UseColor_Specified, specifiedColor=<rgba 0.2396 0.2396 0.2396 1>}
    uVRegion {bIsValid=false, max=<0 0>, min=<0 0>}
   }
   pressedForeground {colorUseRule=UseColor_Foreground, specifiedColor=<rgba 1 0 1 1>}
   pressedPadding=<ltrb 2 2 2 2>
   pressedSlateSound {resourceObject=null}
  }
 }
 ButtonStyle_Selected {
  type=ButtonStyle
  default {
   clickedSlateSound {resourceObject=null}
   disabled {
    bIsDynamicallyLoaded=false
    drawAs=NoDrawType
    imageType=NoImage
    mirroring=NoMirror
    resourceName=None
    resourceObject=null
    tiling=NoTile
    imageSize=<0 0>
    margin=<ltrb 0 0 0 0>
    outlineSettings {
     bUseBrushTransparency=false
     roundingType=HalfHeightRadius
     width=0
     color {colorUseRule=UseColor_Specified, specifiedColor=<rgba 0 0 0 0>}
     cornerRadii=<0 0 0 1>
    }
    tintColor {colorUseRule=UseColor_Specified, specifiedColor=<rgba 0.0397 0.0703 0.099 1>}
    uVRegion {bIsValid=false, max=<0 0>, min=<0 0>}
   }
   disabledForeground {colorUseRule=UseColor_Foreground, specifiedColor=<rgba 1 0 1 1>}
   hovered {
    bIsDynamicallyLoaded=false
    drawAs=Image
    imageType=NoImage
    mirroring=NoMirror
    resourceName=None
    resourceObject=null
    tiling=NoTile
    imageSize=<32 32>
    margin=<ltrb 0 0 0 0>
    outlineSettings {
     bUseBrushTransparency=false
     roundingType=HalfHeightRadius
     width=0
     color {colorUseRule=UseColor_Specified, specifiedColor=<rgba 0 0 0 0>}
     cornerRadii=<0 0 0 1>
    }
    tintColor {colorUseRule=UseColor_Specified, specifiedColor=<rgba 0.1136 0.4089 0.6875 1>}
    uVRegion {bIsValid=false, max=<0 0>, min=<0 0>}
   }
   hoveredForeground {colorUseRule=UseColor_Foreground, specifiedColor=<rgba 1 0 1 1>}
   hoveredSlateSound {resourceObject=null}
   normal {
    bIsDynamicallyLoaded=false
    drawAs=Image
    imageType=NoImage
    mirroring=NoMirror
    resourceName=None
    resourceObject=null
    tiling=NoTile
    imageSize=<32 32>
    margin=<ltrb 0 0 0 0>
    outlineSettings {
     bUseBrushTransparency=false
     roundingType=HalfHeightRadius
     width=0
     color {colorUseRule=UseColor_Specified, specifiedColor=<rgba 0 0 0 0>}
     cornerRadii=<0 0 0 1>
    }
    tintColor {colorUseRule=UseColor_Specified, specifiedColor=<rgba 0.0902 0.3246 0.5457 1>}
    uVRegion {bIsValid=false, max=<0 0>, min=<0 0>}
   }
   normalForeground {colorUseRule=UseColor_Foreground, specifiedColor=<rgba 1 0 1 1>}
   normalPadding=<ltrb 1 1 1 1>
   pressed {
    bIsDynamicallyLoaded=false
    drawAs=Image
    imageType=NoImage
    mirroring=NoMirror
    resourceName=None
    resourceObject=null
    tiling=NoTile
    imageSize=<32 32>
    margin=<ltrb 0 0 0 0>
    outlineSettings {
     bUseBrushTransparency=false
     roundingType=HalfHeightRadius
     width=0
     color {colorUseRule=UseColor_Specified, specifiedColor=<rgba 0 0 0 0>}
     cornerRadii=<0 0 0 1>
    }
    tintColor {colorUseRule=UseColor_Specified, specifiedColor=<rgba 0.043 0.1549 0.2604 1>}
    uVRegion {bIsValid=false, max=<0 0>, min=<0 0>}
   }
   pressedForeground {colorUseRule=UseColor_Foreground, specifiedColor=<rgba 1 0 1 1>}
   pressedPadding=<ltrb 1 1 1 1>
   pressedSlateSound {resourceObject=null}
  }
 }
 currentListItemData {type=Object, default=null}
 doVisualsNeedValidation {type=boolean}
 iconStyle {
  type=SlateBrush
  default {
   bIsDynamicallyLoaded=false
   drawAs=Image
   imageType=NoImage
   mirroring=NoMirror
   resourceName=None
   resourceObject=null
   tiling=NoTile
   imageSize=<32 32>
   margin=<ltrb 0 0 0 0>
   outlineSettings {
    bUseBrushTransparency=false
    roundingType=HalfHeightRadius
    width=0
    color {colorUseRule=UseColor_Specified, specifiedColor=<rgba 0 0 0 0>}
    cornerRadii=<0 0 0 1>
   }
   tintColor {colorUseRule=UseColor_Specified, specifiedColor=<rgba 1 1 1 1>}
   uVRegion {bIsValid=false, max=<0 0>, min=<0 0>}
  }
 }
 isEnabled {type=boolean, default=true}
 isMouseOver {type=boolean}
 isPressed {type=boolean}
 isSelected {type=boolean}
 label {type=string, default=""}
 labelColor_Default {type=LinearColor, default=<rgba 0.0208 0.0208 0.0208 1>}
 labelColor_Selected {type=LinearColor, default=<rgba 0.8906 0.8906 0.8906 1>}
 onClicked {}
 onDoubleClicked {}
 onSelectionChanged {}
 owningList {type=UserWidget, default=null}
]
widgets {info {inheritedWidgetCount=0, parentClass=&UserWidget, rootWidgetClass=null, widgetCount=0}, widgets []}
