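# TA_Generic01_Checkbox is a checkbox-style button widget for tech-art editor tools, derived from TA_BaseEditorButtonWidget.
# It lays out a small image, defaulting to the unchecked texture, beside a CommonText label.
# Getter functions hand the image and text to the parent as its background and label widgets.
# validateVisuals calls the parent, then collapses the label when its string is empty, otherwise making it hit-test invisible.
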WidgetBlueprint /Game/Blueprints/Blutilities/ForTechArt/Widgets/Buttons/TA_Generic01_Checkbox
graphs [
 EventGraph {links [], nodes []}
 getBackgroundWidget {nodes [
   getBackgroundWidget() {is=title K2Node_FunctionEntry}
   > Return Node!(execute, backgroundWidget:Widget Object Reference=Get Image_77() -> (Image_77:Image Object Reference))
  ]}
 getLabelWidget {nodes [
   getLabelWidget() {is=title K2Node_FunctionEntry}
   > Return Node!(execute, labelWidget:Widget Object Reference=Get txtMain() -> (txtMain:Common Text Object Reference))
  ]}
 validateVisuals {nodes [
   validateVisuals() {is=title K2Node_FunctionEntry}
   > Parent: validateVisuals!(execute) -> (then, " ":bool) {is=title K2Node_CallParentFunction}
   n1 = Len(S=Get label() -> (label:String))
   > SetVisibility(
    self=Get txtMain()
    InVisibility=Select(Option 0=Collapsed, Option 1=HitTestInvisible, Index="integer > integer"(A=@n1, B=0) {is=title K2Node_CallFunction})
   )
   Return Node!(execute, " ":bool=false)
  ]}
]
variables []
widgets {
 info {
  inheritedWidgetCount=0
  parentClass=&/TA_BaseEditorButtonWidget.TA_BaseEditorButtonWidget_C
  rootWidgetClass=&Overlay
  widgetCount=4
 }
 widgets [
  Overlay_7 {widgetClassPath=&Overlay, widget=Overlay_7}
  HorizontalBox_0 {
   widgetClassPath=&HorizontalBox
   parent=Overlay_7
   widget=HorizontalBox_0
   slot {class=OverlaySlot, properties {horizontalAlignment=HAlign_Fill}}
  }
  Image_77 {
   widgetClassPath=&Image
   bIsVariable=true
   parent=HorizontalBox_0
   widget=Image_77
   properties {brush {
     bIsDynamicallyLoaded=false
     drawAs=Image
     imageType=NoImage
     mirroring=NoMirror
     resourceName=None
     resourceObject=/Game/Blueprints/Blutilities/ForTechArt/ArtAssets/Textures/Checkbox/TA_Checkbox_Unchecked
     tiling=NoTile
     imageSize=<16 16>
     margin=<ltrb 0.25 0.25 0.25 0.25>
     outlineSettings {
      bUseBrushTransparency=false
      roundingType=HalfHeightRadius
      width=0
      color {colorUseRule=UseColor_Specified, specifiedColor=<rgba 0 0 0 0>}
      cornerRadii=<0 0 0 1>
     }
     tintColor {colorUseRule=UseColor_Specified, specifiedColor=<rgba 1 1 1 1>}
     uVRegion {bIsValid=false, max=<0 0>, min=<0 0>}
    }}
   slot {class=HorizontalBoxSlot, properties {verticalAlignment=VAlign_Center, padding=<ltrb 0 0 3 0>}}
  }
  txtMain {
   widgetClassPath=&CommonTextBlock
   bIsVariable=true
   parent=HorizontalBox_0
   widget=txtMain
   properties {
    text=Text Block
    font {
     bForceMonospaced=false
     bMaterialIsStencil=false
     fontMaterial=null
     fontObject=&NotoSans
     letterSpacing=0
     monospacedWidth=1
     size=10.5
     skewAmount=0
     typefaceFontName=NotoSans-Regular
     outlineSettings {
      bApplyOutlineToDropShadows=false
      bMiteredCorners=false
      bSeparateFillAlpha=false
      outlineMaterial=null
      outlineSize=0
      outlineColor=<rgba 0 0 0 1>
     }
    }
    shadowOffset=<0 0>
   }
   slot {class=HorizontalBoxSlot, properties {verticalAlignment=VAlign_Center}}
  }
 ]
}
